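# Translucent, unlit, two-sided material for the FearlessFlightHero pickaxe trail, flagged for Niagara ribbons.
# A body mask texture feeds colors and hue into SleepTimeFxFunction and selects per-region panner speeds for a streak texture.
# The function's result, tinted by particle color and the streaks, drives emissive.
# A rounded flicker mask multiplies streaks by glitch patterns blended by sine and cosine time waves, offset by a Random dynamic parameter.
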Material /BRCosmetics/Weapons/FORT_Melee/Pickaxe_FearlessFlightHero/FX/M_FFH_Trail
graph {
 properties {
  bUsedWithNiagaraRibbons=true
  blendMode=BLEND_Translucent
  pixelDepthOffsetMode=PDOM_Legacy
  shadingModel=MSM_Unlit
  twoSided=true
  usedShadingModels=Unlit
 }
 nodes [
  ComponentMask(None=TextureCoordinate()) {r=true}
  n1 = ParticleColor()
  n2 = Constant() {r=1}
  n3 = ScalarParameter() {defaultValue=0.4889, parameterName=Blue}
  n4 = MaterialFunctionCall(XY=TextureCoordinate()) {materialFunction=&Swizzle}
  n5 = TextureSample(UVs=@n4.YX) {samplerType=SAMPLERTYPE_Masks, texture=/Game/Characters/Player/Male/Medium/Bodies/M_MED_SleepyTime/Textures/T_MED_SleepTime_Body_Mask01}
  n6 = VectorParameter() {parameterName=magenta ColorA, defaultValue=<rgba 1 0 0.1111 1>}
  n7 = MaterialFunctionCall!(
   Exponen=ScalarParameter() {defaultValue=0.9324, group=outlineCtrl, parameterName=Exponen}
   Hue=LinearInterpolate(A=LinearInterpolate(A=@n2, B=@n3, Alpha=@n5.G), B=ScalarParameter() {defaultValue=0.7, parameterName=purple}, Alpha=@n5.B)
   InsideMaincolor=LinearInterpolate(A=@n6.RGB, B=VectorParameter() {parameterName=insideMain, defaultValue=<rgba 1 0.7732 0.0645 1>}.RGB, Alpha=@n5.R)
   InsideDarkColor=LinearInterpolate(A=@n6.RGB, B=VectorParameter() {parameterName=InsideDark, defaultValue=<rgba 0.4097 0.1339 0.0433 1>}.RGB, Alpha=@n5.R)
   outline color=LinearInterpolate(A=@n6.RGB, B=VectorParameter() {group=outlineCtrl, parameterName=outline color, defaultValue=<rgba 0.7656 0.6165 0.1979 1>}.RGB, Alpha=@n5.R)
   BaseRef=ScalarParameter() {defaultValue=0.0643, group=outlineCtrl, parameterName=BaseRef}
  ) {materialFunction=/Game/Weapons/WeaponSkins/MaterialFunction/SleepTimeFxFunction}
  n8 = LinearInterpolate(
   A=LinearInterpolate(A=Panner(Coordinate=@n4.YX) {speedY=0.35}, B=Panner(Coordinate=@n4.YX) {speedY=0.1}, Alpha=@n5.R)
   B=Panner(Coordinate=@n4.YX) {speedY=0.7}
   Alpha=@n5.B
  )
  n9 = Panner(Coordinate=@n4.YX) {speedY=1.1}
  n10 = Constant() {r=0.3}
  n11 = Constant() {r=1}
  n12 = TextureSample(UVs=Multiply(A=LinearInterpolate(A=@n8, B=@n9, Alpha=@n5.G), B=AppendVector(A=@n10, B=@n11))) {texture=&/T_MED_test2}
  n13 = LinearInterpolate(A=Constant3Vector() {constant=<rgba 1 0 0.7629 1>}, B=Constant3Vector() {constant=<rgba 0.2584 2 0 0>}, Alpha=@n12.G)
  MaterialFunctionCall(Emissive=Multiply(A=Multiply(A=@n1.RGB, B=@n7.Result), B=Add(A=@n13) {constB=0.1})) {materialFunction=&MF_FX_EmissiveAdjustments}
  n14 = DynamicParameter() {paramNames [Random,Param2,Param3,Param4]}
  n15 = Multiply(A=TextureCoordinate(), B=LinearInterpolate(Alpha=@n14.Random) {constA=1, constB=2})
  n16 = OneMinus(None=TextureSample(UVs=MaterialFunctionCall(UVs=@n15) {materialFunction=&ScaleUVsByCenter}.UVs) {texture=&/T_Glitch_Pattern_FFH}.R)
  n17 = Time()
  n18 = Add(A=@n17, B=@n14.Random)
  n19 = Multiply(A=Add(A=Sine(None=@n18))) {constB=0.5}
  n20 = Add(A=@n17, B=@n14.Random)
  n21 = Multiply(A=Add(A=Cosine(None=@n20))) {constB=0.5}
  n22 = Multiply(A=@n19, B=@n21)
  n23 = LinearInterpolate(A=@n16, Alpha=@n22)
  n24 = TextureSample() {samplerSource=SSM_Clamp_WorldGroupSettings, texture=&T_Glitch_Pattern_FFH}
  n25 = Step(Y=OneMinus(None=@n24.G), X=Multiply(A=Add(A=@n19) {constB=-0.5}) {constB=1.5})
  Round(None=Multiply(A=@n12.R, B=LinearInterpolate(A=@n23, B=@n25, Alpha=@n22)))
  Step(Y=@n16, X=@n21)
 ]
}
