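# Unlit, two-sided translucent Niagara ribbon material for the Fearless Flight Hero pickaxe trail.
# A borrowed SleepyTime body mask splits the trail into bands that pan the bits texture at different speeds and pick color pairs.
# Particle color tints the bands before emissive adjustment.
# A glitch pattern, animated by sine and cosine time offset by a Random dynamic parameter, multiplies the bits and is rounded into a hard-edged mask.
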Material /BRCosmetics/Weapons/FORT_Melee/Pickaxe_FearlessFlightHero/FX/M_FFH_Trail_2
graph {
 properties {
  bUsedWithNiagaraRibbons=true
  blendMode=BLEND_Translucent
  pixelDepthOffsetMode=PDOM_Legacy
  shadingModel=MSM_Unlit
  twoSided=true
  usedShadingModels=Unlit
 }
 nodes [
  n1 = MaterialFunctionCall(XY=TextureCoordinate()) {materialFunction=&Swizzle}
  n2 = TextureSample(UVs=@n1.YX) {samplerType=SAMPLERTYPE_Masks, texture=/Game/Characters/Player/Male/Medium/Bodies/M_MED_SleepyTime/Textures/T_MED_SleepTime_Body_Mask01}
  n3 = LinearInterpolate(A=Panner(Coordinate=@n1.YX) {speedY=0.35}, B=Panner(Coordinate=@n1.YX) {speedY=0.2}, Alpha=@n2.B)
  n4 = Panner(Coordinate=@n1.YX) {speedY=0.7}
  n5 = TextureSample(UVs=LinearInterpolate(A=LinearInterpolate(A=@n3, B=@n4, Alpha=@n2.G), B=Panner(Coordinate=@n1.YX) {speedY=1.1}, Alpha=@n2.R)) {texture=&/T_FFH_Trail_Bits_Mask2}
  Add(A=LinearInterpolate(A=Constant3Vector() {constant=<rgba 1 0 0.7629 1>}, B=Constant3Vector() {constant=<rgba 0.2584 2 0 0>}, Alpha=@n5.G)) {constB=0.1}
  n6 = Constant3Vector() {constant=<rgba 1 0.1271 0 1>}
  n7 = Constant3Vector() {constant=<rgba 0 1 0.0657 1>}
  n8 = OneMinus(None=@n2.R)
  n9 = LinearInterpolate(A=Constant3Vector() {constant=<rgba 0 0.75 0.5418 1>}, B=Constant3Vector() {constant=<rgba 0 0.5 0.015 1>}, Alpha=@n5.G)
  n10 = Constant3Vector() {constant=<rgba 0 1 0.0957 1>}
  n11 = Constant3Vector() {constant=<rgba 2 1.8208 0 1>}
  n12 = Add(
   A=LinearInterpolate(B=LinearInterpolate(A=@n6, B=@n7, Alpha=@n5.G), Alpha=@n2.R)
   B=LinearInterpolate(A=Multiply(A=@n8, B=@n9), B=LinearInterpolate(A=@n10, B=@n11, Alpha=@n5.G), Alpha=@n2.G)
  )
  MaterialFunctionCall(Emissive=Multiply(A=ParticleColor().RGB, B=Multiply(A=@n12))) {materialFunction=&MF_FX_EmissiveAdjustments}
  n13 = DynamicParameter() {paramNames [Random,Param2,Param3,Param4]}
  n14 = Multiply(A=TextureCoordinate(), B=LinearInterpolate(Alpha=@n13.Random) {constA=1, constB=2})
  n15 = OneMinus(None=TextureSample(UVs=MaterialFunctionCall(UVs=@n14) {materialFunction=&ScaleUVsByCenter}.UVs) {texture=&/T_Glitch_Pattern_FFH}.R)
  n16 = Time()
  n17 = Add(A=@n16, B=@n13.Random)
  n18 = Multiply(A=Add(A=Sine(None=@n17))) {constB=0.5}
  n19 = Add(A=@n16, B=@n13.Random)
  n20 = Multiply(A=Add(A=Cosine(None=@n19))) {constB=0.5}
  n21 = Multiply(A=@n18, B=@n20)
  n22 = LinearInterpolate(A=@n15, Alpha=@n21)
  n23 = TextureSample() {samplerSource=SSM_Clamp_WorldGroupSettings, texture=&T_Glitch_Pattern_FFH}
  n24 = Step(Y=OneMinus(None=@n23.G), X=Multiply(A=Add(A=@n18) {constB=-0.5}) {constB=1.5})
  Round(None=Multiply(A=@n5.R, B=LinearInterpolate(A=@n22, B=@n24, Alpha=@n21)))
  Step(Y=@n15, X=@n20)
 ]
}
