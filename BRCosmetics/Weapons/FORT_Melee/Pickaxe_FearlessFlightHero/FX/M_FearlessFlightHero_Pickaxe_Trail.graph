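# Translucent, unlit, two-sided Niagara ribbon material for the Fearless Flight Hero pickaxe trail.
# A green-to-red gradient along U passes through MF_FX_EmissiveAdjustments for color.
# A mask uses Ceil and Round to quantize UV ramps and a panning, parameterized "Beat" texture into hard-edged steps.
# The inverted steps are added to a tiled, contrast-boosted CircleFade pattern, weighted toward the middle of U.
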Material /BRCosmetics/Weapons/FORT_Melee/Pickaxe_FearlessFlightHero/FX/M_FearlessFlightHero_Pickaxe_Trail
graph {
 properties {
  bAutomaticallySetUsageInEditor=false
  bUsedWithNiagaraRibbons=true
  bUsedWithStaticMesh=false
  blendMode=BLEND_Translucent
  pixelDepthOffsetMode=PDOM_Legacy
  shadingModel=MSM_Unlit
  twoSided=true
  usedShadingModels=Unlit
 }
 nodes [
  n1 = Constant3Vector() {constant=<rgba 0.1145 0.8958 0 1>}
  n2 = Constant3Vector() {constant=<rgba 1 0 0 1>}
  n3 = TextureCoordinate()
  n4 = Multiply(A=Add(A=ComponentMask(None=@n3) {r=true}) {constB=0.05}) {constB=2}
  Reroute(None=MaterialFunctionCall(Emissive=LinearInterpolate(A=@n1, B=@n2, Alpha=@n4)) {materialFunction=&MF_FX_EmissiveAdjustments}.Result)
  n5 = TextureCoordinate()
  n6 = ComponentMask(None=@n5) {r=true}
  n7 = Constant() {r=16}
  n8 = Divide(A=Ceil(None=Multiply(A=@n6, B=@n7) {constB=8}), B=@n7) {constB=8}
  n9 = TextureCoordinate() {uTiling=0, vTiling=6}
  n10 = Constant2Vector() {g=1, r=1}
  n11 = Divide(A=Ceil(None=Multiply(A=@n9, B=@n10)), B=@n7) {constB=8}
  n12 = TextureSampleParameter2D!(
   UVs=Panner(Coordinate=MaterialFunctionCall(XY=@n11) {materialFunction=&Swizzle}.YX) {speedX=0.15, speedY=0.1}
   Level
   Apply View MipBias
  ) {
   group=Beat Texture
   mipValueMode=TMVM_MipLevel
   parameterName=Beat Texture Input
   sortPriority=-5
   texture=&/T_FFH_Anim
  }
  n13 = Divide(A=Ceil(None=Multiply(A=@n12.R) {constB=16})) {constB=16}
  n14 = ComponentMask(None=@n5) {r=true}
  n15 = Add(A=Multiply(A=Multiply(A=@n8, B=@n13), B=Multiply(A=@n14, B=@n14)), B=Constant() {r=0.925})
  n16 = TextureCoordinate()
  n17 = Constant2Vector() {g=0.75, r=0.1}
  n18 = OneMinus(None=ComponentMask(None=Add(A=@n16, B=@n17)) {r=true})
  n19 = Multiply(A=Multiply(A=Power(Base=@n18) {constExponent=8}, B=@n12.R)) {constB=5}
  n20 = Add(A=Ceil(None=Add(A=@n15) {constB=-0.955}), B=Round(None=Clamp(None=@n19))) {constB=0}
  n21 = ComponentMask(None=TextureCoordinate()) {r=true}
  n22 = Multiply(A=@n21, B=OneMinus(None=@n21))
  n23 = TextureSample(UVs=Multiply(A=TextureCoordinate(), B=Constant2Vector() {g=1, r=6})) {samplerType=SAMPLERTYPE_LinearColor, texture=/Game/Weapons/WeaponSkins/Wraps/Textures/CircleFade/T_Wrap_CircleFade_M}
  n24 = Constant() {r=0.5}
  n25 = Saturate(None=Multiply(A=MaterialFunctionCall(In=@n23.G, Contrast=@n24) {materialFunction=&CheapContrast}.Result) {constB=10})
  Reroute(None=Add(A=OneMinus(None=@n20), B=Multiply(A=@n22, B=@n25)))
 ]
}
